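# Translucent unlit, two-sided Niagara material that draws a starfield and nebula for the WishingStar emote.
# Emissive sums panning BumpOffset parallax star layers, modulated by two twinkle masks, with a tinted nebula, scaled by particle color.
# Opacity combines a noisy radial mask, dynamic-parameter erosion and particle alpha, with depth fade and dimming at vertical view angles.
# Rotated, dynamic-speed shooting-star branches exist but do not reach the emissive adjustment chain.
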Material /BRCosmetics/Animation/Game/MainPlayer/Emotes/WishingStar/FX/M_Emote_WishingStar_Starfield
graph {
 properties {
  bAutomaticallySetUsageInEditor=false
  bUsedWithNiagaraMeshParticles=true
  bUsedWithNiagaraSprites=true
  blendMode=BLEND_Translucent
  shadingModel=MSM_Unlit
  twoSided=true
  usedShadingModels=Unlit
 }
 nodes [
  n1 = Add(A=TextureCoordinate()) {constB=-0.5}
  n2 = AppendVector(
   A=ScalarParameter() {defaultValue=1, group=Star, parameterName=Small Star Tile X}
   B=ScalarParameter() {defaultValue=1, group=Star, parameterName=Small Star Tile Y}
  )
  n3 = Panner(
   Coordinate=Add(A=Multiply(A=@n1, B=@n2)) {constB=0.5}
   Speed=AppendVector(A=ScalarParameter() {group=Star, parameterName=Small Star Speed X}, B=ScalarParameter() {group=Star, parameterName=Small Star Speed Y})
  )
  n4 = ScalarParameter() {defaultValue=-50, group=Star, parameterName=Small Star Height}
  n5 = Multiply(
   A=TextureSample(UVs=BumpOffset(Coordinate=@n3, Height=@n4)) {texture=/Engine/MapTemplates/Sky/T_Sky_Stars}.RGB
   B=ScalarParameter() {defaultValue=1, group=Star, parameterName=Small Star Brightness}
  )
  n6 = TextureCoordinate()
  n7 = ScalarParameter() {defaultValue=1, group=Star, parameterName=Big Star Tile X}
  n8 = ScalarParameter() {defaultValue=1, group=Star, parameterName=Big Star Tile Y}
  n9 = Add(A=Multiply(A=Add(A=@n6) {constB=-0.5}, B=AppendVector(A=@n7, B=@n8))) {constB=0.5}
  n10 = AppendVector(A=ScalarParameter() {group=Star, parameterName=Big Star Speed X}, B=ScalarParameter() {group=Star, parameterName=Big Star Speed Y})
  n11 = TextureSample(UVs=BumpOffset(Coordinate=Panner(Coordinate=@n9, Speed=@n10), Height=ScalarParameter() {defaultValue=-20, group=Star, parameterName=Big Star Height})) {samplerType=SAMPLERTYPE_Masks, texture=/Game/Characters/Player/Male/Medium/Bodies/M_Med_Soldier_04/Skins/Galaxy/Textures/Celestial_T_Stars}
  n12 = Multiply(A=@n11.R, B=ScalarParameter() {defaultValue=1, group=Star, parameterName=Big Star Brightness})
  n13 = Multiply(
   A=Add(A=TextureCoordinate()) {constB=-0.5}
   B=AppendVector(
    A=ScalarParameter() {defaultValue=1, group=Star, parameterName=Twinkle Tile X}
    B=ScalarParameter() {defaultValue=1, group=Star, parameterName=Twinkle Tile Y}
   )
  )
  n14 = ScalarParameter() {group=Star, parameterName=Twinkle Speed X}
  n15 = ScalarParameter() {group=Star, parameterName=Twinkle Speed Y}
  n16 = TextureSample(UVs=Panner(Coordinate=Add(A=@n13) {constB=0.5}, Speed=AppendVector(A=@n14, B=@n15))) {samplerType=SAMPLERTYPE_Masks, texture=&Celestial_T_Stars}
  n17 = Multiply(
   A=Add(A=TextureCoordinate()) {constB=-0.5}
   B=AppendVector(
    A=ScalarParameter() {defaultValue=1, group=Shooting Star 2, parameterName=Twinkle 2 Tile X}
    B=ScalarParameter() {defaultValue=1, group=Shooting Star 2, parameterName=Twinkle 2 Tile Y}
   )
  )
  n18 = ScalarParameter() {group=Shooting Star 2, parameterName=Twinkle 2 Speed X}
  n19 = ScalarParameter() {group=Shooting Star 2, parameterName=Twinkle 2 Speed Y}
  n20 = TextureSample(UVs=Panner(Coordinate=Add(A=@n17) {constB=0.5}, Speed=AppendVector(A=@n18, B=@n19))) {samplerType=SAMPLERTYPE_Masks, texture=&Celestial_T_Stars}
  n21 = Multiply(
   A=Multiply(A=Add(A=@n5, B=@n12), B=Multiply(A=@n16.G, B=@n20.G))
   B=VectorParameter() {group=Star, parameterName=Star Color, defaultValue=<rgba 1 1 1 1>}.RGB
  )
  n22 = MaterialFunctionCall(UVs=TextureCoordinate(), "Rotation Angle (0-1)"=ScalarParameter() {group=Shooting Star, parameterName=Shooting Star Rotation}) {materialFunction=&CustomRotator}
  n23 = Multiply(
   A=Add(A=@n22.Rotated Values) {constB=-0.5}
   B=AppendVector(
    A=ScalarParameter() {defaultValue=1, group=Shooting Star, parameterName=Shooting Star Tile X}
    B=ScalarParameter() {defaultValue=1, group=Shooting Star, parameterName=Shooting Star Tile Y}
   )
  )
  n24 = DynamicParameter() {defaultValue=<rgba 0 0 1 1>, paramNames [Erosion Edge, Erosion Feather, Shooting Star 1 Speed Mult, Shooting Star 2 Speed Mult]}
  n25 = Multiply(A=ScalarParameter() {group=Shooting Star, parameterName=Shooting Star Speed X}, B=@n24.Shooting Star 1 Speed Mult)
  n26 = Multiply(A=ScalarParameter() {group=Shooting Star, parameterName=Shooting Star Speed Y}, B=@n24.Shooting Star 1 Speed Mult)
  n27 = TextureSampleParameter2D(UVs=Panner(Coordinate=Add(A=@n23) {constB=0.5}, Speed=AppendVector(A=@n25, B=@n26))) {
   group=Shooting Star
   parameterName="Shooting Star Texture (R)"
   texture=&/T_Emote_WishingStar_ShootingStar
  }
  n28 = Multiply(A=@n27.R, B=ScalarParameter() {defaultValue=1, group=Shooting Star, parameterName=Shooting Star Intensity})
  n29 = Multiply(
   A=Add(A=@n22.Rotated Values) {constB=-0.5}
   B=AppendVector(
    A=ScalarParameter() {defaultValue=1, group=Shooting Star, parameterName=Shooting Star Mask Tile X}
    B=ScalarParameter() {defaultValue=1, group=Shooting Star, parameterName=Shooting Star Mask Tile Y}
   )
  )
  n30 = ScalarParameter() {group=Shooting Star, parameterName=Shooting Star Mask Speed X}
  n31 = Multiply(A=@n24.Shooting Star 1 Speed Mult, B=ScalarParameter() {group=Shooting Star, parameterName=Shooting Star Mask Speed Y})
  n32 = TextureSampleParameter2D(UVs=Panner(Coordinate=Add(A=@n29) {constB=0.5}, Speed=AppendVector(A=@n30, B=@n31))) {
   group=Shooting Star
   parameterName="Shooting Star Mask Texture (G)"
   texture=&T_Emote_WishingStar_ShootingStar
  }
  n33 = Multiply(
   A=Power(Base=@n32.G, Exp=ScalarParameter() {defaultValue=1, group=Shooting Star, parameterName=Shooting Star Mask Power})
   B=ScalarParameter() {defaultValue=1, group=Shooting Star, parameterName=Shooting Star Mask Mult}
  )
  n34 = DynamicParameter() {desc=Shooting Star Color, parameterIndex=1, defaultValue=<rgba 0 0 1 1>}
  n35 = Multiply(
   A=Add(A=@n22.Rotated Values) {constB=-0.5}
   B=AppendVector(
    A=ScalarParameter() {defaultValue=1, group=Shooting Star 2, parameterName=Shooting Star 2 Tile X}
    B=ScalarParameter() {defaultValue=1, group=Shooting Star 2, parameterName=Shooting Star 2 Tile Y}
   )
  )
  n36 = DynamicParameter() {defaultValue=<rgba 0 0 1 1>, paramNames [Erosion Edge, Erosion Feather, Shooting Star 1 Speed Mult, Shooting Star 2 Speed Mult]}
  n37 = Multiply(A=ScalarParameter() {group=Shooting Star 2, parameterName=Shooting Star 2 Speed X}, B=@n36.Shooting Star 1 Speed Mult)
  n38 = Multiply(A=ScalarParameter() {group=Shooting Star 2, parameterName=Shooting Star 2 Speed Y}, B=@n36.Shooting Star 1 Speed Mult)
  n39 = TextureSampleParameter2D(UVs=Panner(Coordinate=Add(A=@n35) {constB=0.5}, Speed=AppendVector(A=@n37, B=@n38))) {
   group=Shooting Star 2
   parameterName="Shooting Star 2 Texture (R)"
   texture=&T_Emote_WishingStar_ShootingStar
  }
  n40 = Multiply(
   A=Add(A=@n22.Rotated Values) {constB=-0.5}
   B=AppendVector(
    A=ScalarParameter() {defaultValue=1, group=Shooting Star 2, parameterName=Shooting Star 2 Mask Tile X}
    B=ScalarParameter() {defaultValue=1, group=Shooting Star 2, parameterName=Shooting Star 2 Mask Tile Y}
   )
  )
  n41 = Multiply(A=ScalarParameter() {group=Shooting Star 2, parameterName=Shooting Star 2 Mask Speed X}, B=@n36.Shooting Star 2 Speed Mult)
  n42 = ScalarParameter() {group=Shooting Star 2, parameterName=Shooting Star 2 Mask Speed Y}
  n43 = TextureSampleParameter2D(UVs=Panner(Coordinate=Add(A=@n40) {constB=0.5}, Speed=AppendVector(A=@n41, B=@n42))) {
   group=Shooting Star 2
   parameterName="Shooting Star 2 Mask Texture (G)"
   texture=&T_Emote_WishingStar_ShootingStar
  }
  n44 = ScalarParameter() {defaultValue=1, group=Shooting Star 2, parameterName=Shooting Star 2 Mask Power}
  n45 = Multiply(
   A=Multiply(A=@n39.R, B=ScalarParameter() {defaultValue=1, group=Shooting Star 2, parameterName=Shooting Star 2 Intensity})
   B=Multiply(A=Power(Base=@n43.G, Exp=@n44), B=ScalarParameter() {defaultValue=1, group=Shooting Star 2, parameterName=Shooting Star 2 Mask Mult})
  )
  n46 = Add(A=Multiply(A=Multiply(A=@n28, B=@n33), B=@n34.RGB), B=Multiply(A=Multiply(A=@n45, B=@n34.RGB)) {constB=1.65})
  Add(A=@n21, B=@n46)
  n47 = Multiply(
   A=Add(A=TextureCoordinate()) {constB=-0.5}
   B=AppendVector(
    A=ScalarParameter() {defaultValue=1, group=Galaxy, parameterName=Galaxy Tile X}
    B=ScalarParameter() {defaultValue=1, group=Galaxy, parameterName=Galaxy Tile Y}
   )
  )
  n48 = ScalarParameter() {group=Galaxy, parameterName=Galaxy Speed X}
  n49 = ScalarParameter() {group=Galaxy, parameterName=Galaxy Speed Y}
  n50 = TextureSampleParameter2D(UVs=Panner(Coordinate=Add(A=@n47) {constB=0.5}, Speed=AppendVector(A=@n48, B=@n49))) {
   group=Galaxy
   parameterName=Galaxy Texture
   texture=/Game/Characters/Player/Male/Medium/Bodies/M_Med_Soldier_04/Skins/Galaxy/Textures/T_FN_Nebula
  }
  n51 = ScalarParameter() {defaultValue=1, group=Galaxy, parameterName=Galaxy Intensity}
  n52 = Desaturation(
   None=Multiply(A=VectorParameter() {group=Galaxy, parameterName=Galaxy Color, defaultValue=<rgba 1 1 1 0>}.RGB, B=Multiply(A=@n50.RGB, B=@n51))
   Fraction=ScalarParameter() {group=Galaxy, parameterName=Saturation}
  )
  n53 = ParticleColor()
  MaterialFunctionCall(Emissive=Multiply(A=Add(A=@n52, B=@n21), B=@n53.RGB)) {materialFunction=&MF_FX_EmissiveAdjustments}
  n54 = Multiply(
   A=Add(A=TextureCoordinate()) {constB=-0.5}
   B=AppendVector(
    A=ScalarParameter() {defaultValue=1, group=Shooting Star 2, parameterName=Twinkle 2 Tile X}
    B=ScalarParameter() {defaultValue=1, group=Shooting Star 2, parameterName=Twinkle 2 Tile Y}
   )
  )
  n55 = ScalarParameter() {group=Shooting Star 2, parameterName=Twinkle 2 Speed X}
  n56 = ScalarParameter() {group=Shooting Star 2, parameterName=Twinkle 2 Speed Y}
  n57 = TextureSampleParameter2D(UVs=Panner(Coordinate=Add(A=@n54) {constB=0.5}, Speed=AppendVector(A=@n55, B=@n56))) {
   group=Shooting Star 2
   parameterName=Twinkle 2 Texture
   samplerType=SAMPLERTYPE_Masks
   texture=&Celestial_T_Stars
  }
  Multiply(A=@n46, B=@n57.G)
  Multiply(A=@n30, B=@n24.Shooting Star 1 Speed Mult)
  Multiply(A=@n36.Shooting Star 2 Speed Mult, B=@n42)
  n58 = DynamicParameter() {defaultValue=<rgba 0 0 1 1>, paramNames [Erosion Edge, Erosion Feather, Shooting Star 1 Speed Mult, Shooting Star 2 Speed Mult]}
  n59 = Add(A=@n58.Erosion Edge, B=@n58.Erosion Edge)
  n60 = Add(A=TextureCoordinate()) {constB=-0.5}
  n61 = AppendVector(
   A=ScalarParameter() {defaultValue=1, group=Erosion, parameterName=Erosion Tile X}
   B=ScalarParameter() {defaultValue=1, group=Erosion, parameterName=Erosion Tile Y}
  )
  n62 = Panner(
   Coordinate=Add(A=Multiply(A=@n60, B=@n61)) {constB=0.5}
   Speed=AppendVector(A=ScalarParameter() {group=Erosion, parameterName=Erosion Speed X}, B=ScalarParameter() {group=Erosion, parameterName=Erosion Speed Y})
  )
  n63 = Multiply(
   A=Power(
    Base=TextureSampleParameter2D(UVs=@n62) {group=Erosion, parameterName=Erosion Texture, texture=/Game/Effects/Fort_Effects/Textures/T_Tile_Pattern_01}.R
    Exp=ScalarParameter() {defaultValue=1, group=Erosion, parameterName=Erosion Power}
   )
   B=MaterialFunctionCall(
    CenterPosition=Constant2Vector() {g=0.5, r=0.5}
    Radius=ScalarParameter() {defaultValue=0.5, group=Erosion, parameterName=Erosion Mask Radius}
    Density=ScalarParameter() {defaultValue=1, group=Erosion, parameterName=Erosion Mask Density}
   ) {materialFunction=&RadialGradientExponential}.RadialGradientExponential
  )
  n64 = MaterialFunctionCall(
   UVs=TextureCoordinate()
   CenterPosition=Constant2Vector() {g=0.5, r=0.5}
   Radius=ScalarParameter() {defaultValue=0.5, group=Mask, parameterName=SphereMask Radius}
   Density=ScalarParameter() {defaultValue=3, group=Mask, parameterName=SphereMask Density}
  ) {materialFunction=&RadialGradientExponential}
  n65 = TextureCoordinate()
  n66 = MaterialFunctionCall(Vector or UVs=Add(A=@n65) {constB=-0.5}) {materialFunction=&VectorToRadialValue}
  n67 = AppendVector(
   A=ScalarParameter() {defaultValue=1, group=Mask, parameterName=Noise Tile}
   B=ScalarParameter() {defaultValue=1, group=Mask, parameterName=Noise Scale}
  )
  n68 = ScalarParameter() {group=Mask, parameterName=Noise Speed X}
  n69 = ScalarParameter() {group=Mask, parameterName=Noise Speed Y}
  n70 = TextureSampleParameter2D!(
   UVs=Panner(Coordinate=Multiply(A=@n66.Radial Coordinates, B=@n67), Speed=AppendVector(A=@n68, B=@n69))
   "DDX(UVs)"=DDX(Value=@n65)
   "DDY(UVs)"=DDY(Value=@n65)
   Apply View MipBias
  ) {
   group=Mask
   mipValueMode=TMVM_Derivative
   parameterName=Noise Texture
   samplerType=SAMPLERTYPE_LinearGrayscale
   texture=/Game/Global/Textures/Default/Caustics/T_Nosie_Caustics_Grayscale
  }
  n71 = Max(A=@n64.RadialGradientExponential, B=@n70.RGB)
  n72 = Multiply(A=@n53.A, B=Multiply(A=SmoothStep(Min=@n58.Erosion Edge, Max=@n59, Value=@n63), B=Multiply(A=@n64.RadialGradientExponential, B=@n71)))
  n73 = ScalarParameter() {defaultValue=40, parameterName=FadeDistance}
  n74 = Abs(None=CameraVectorWS())
  n75 = Constant3Vector() {constant=<rgba 0 0 1 0>}
  n76 = ScalarParameter() {defaultValue=0.3, parameterName=Fade Low}
  n77 = ScalarParameter() {defaultValue=1, parameterName=Fade High}
  n78 = Clamp(None=Power(Base=DotProduct(A=@n74, B=@n75), Exp=ScalarParameter() {defaultValue=2, parameterName=FadeSens}), Min=@n76, Max=@n77) {minDefault=0.3}
  Saturate(None=Multiply(
    A=DepthFade(Opacity=@n72, FadeDistance=@n73)
    B=MaterialFunctionCall(Input=@n78, Input Low=@n76, Input High=@n77, Target Low=@n77, Target High=@n76) {materialFunction=&RemapValueRange}.Result
   ))
 ]
}
