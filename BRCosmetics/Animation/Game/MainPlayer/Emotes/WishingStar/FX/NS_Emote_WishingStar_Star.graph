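# Niagara system for the Wishing Star player emote: glowing stars, shooting comets, trails and sparkle bursts.
# ShootingStarHead launches mesh stars with comet sounds, and ShootingStarRibbon reads its particles to draw their tails.
# Pickup and trail emitters sample the skeletal mesh, with rates scaled by User parameters such as StarParticles, Trail1 and Trail2.
# Timed bursts use spawn-group masks to layer glows, sparkles and rays within shared emitters.
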NiagaraSystem /BRCosmetics/Animation/Game/MainPlayer/Emotes/WishingStar/FX/NS_Emote_WishingStar_Star
dependencies {
 usedDataInterfaces [
  &NiagaraDataInterfaceCurve
  &NiagaraDataInterfaceColorCurve
  &NiagaraDataInterfaceVectorCurve
  &NiagaraDataInterfaceAudioPlayer
  &NiagaraDataInterfaceParticleRead
  &NiagaraDataInterfaceSkeletalMesh
 ]
 usedDynamicInputs [&FNTimeOfDayColor]
 usedModules [
  &self:SystemSpawnScript.SystemScriptSource.SystemScriptGraph.NiagaraNodeAssignment_0.SetVariables_#1
  &SystemState
  &self:SystemSpawnScript.SystemScriptSource.SystemScriptGraph.NiagaraNodeAssignment_1.SetVariables_#2
  &self:SystemSpawnScript.SystemScriptSource.SystemScriptGraph.NiagaraNodeAssignment_2.SetVariables_#3
 ]
}
emitters {
 Galaxy {rendererClasses [], renderers []}
 ShootingStarHead {
  emitterSpawnScript [&InitializeAudioPlayer(Audio Player=DataInterface() {propertyValues {
      Attenuation=null
      Concurrency=null
      MaxPlaysPerTick=10
      SoundToPlay=/BRCosmetics/Sounds/Emotes/WishingStar/MSS_Emote_Foley_WishingStar_CometBys
      bAllowLoopingOneShotSounds=false
      bLimitPlaysPerTick=true
      bOnlyActiveDuringGameplay=false
      bStopWhenComponentIsDestroyed=true
      ConfigurationUserParameter {parameter {name=None, typeDefHandle {registeredTypeIndex=1}, varData []}}
      ParameterNames []
     }})]
  emitterUpdateScript [
   &EmitterState()
   &SpawnRate(
    Spawn Probability=0.6
    SpawnRate=&FloatFromCurve(
     FloatCurve=DataInterface() {propertyValues {Curve {keys [<at 0 0>,<at 0.075 1>,<at 0.3 1>,<at 0.5 0.7>,<at 0.75 0.7>,<at 0.8 0>]}}}
     CurveIndex=$Emitter.NormalizedLoopAge
     Scale Curve=3.35
    )
    Use Spawn Probability=true
   )
  ]
  particleSpawnScript [
   &InitializeParticle(
    Color=<rgba 5 4.0865 0 1>
    Lifetime Max=1
    Lifetime Min=0.7
    Lifetime Mode=::NewEnumerator1
    Mesh Scale Mode=::NewEnumerator1
    Mesh Uniform Scale Max=0.6
    Mesh Uniform Scale Min=0.4
    Position Offset=<90 -75 160>
    Sprite Size Mode=::NewEnumerator1
    Uniform Sprite Size Max=25
    Uniform Sprite Size Min=18
    UsePositionOffset=true
    Write Material Random=true
   )
   &InitialMeshOrientation(Mesh Orientation Mode=::NewEnumerator1, Rotation=<0 0 0.25>, Use Rotation Vector=true)
   &ShapeLocation(MaskSpawnGroup=false, Non Uniform Scale=<0.3 2 1.5>, Override Seed=false, Sphere Radius=25, UseOwnerScale=true)
   &AddVelocity(Velocity=&RandomRangeVector(Minimum=<-350 120 0>, Maximum=<-480 150 0>))
   /Niagara/Modules/Audio/PlayAudio()
  ]
  particleUpdateScript [
   &ParticleState()
   &SpriteRotationRate(Rotation Rate=-3)
   &ScaleMeshSize(Scale Factor=&VectorFromFloat(Value=&FloatFromCurve(FloatCurve=DataInterface() {propertyValues {Curve {
         defaultValue=0
         keys [
          {arriveTangent=9.0725, interpMode=RCIM_Cubic, leaveTangent=9.0725, tangentMode=RCTM_User, time=0, value=0}
          {arriveTangent=-2.5249, interpMode=RCIM_Cubic, leaveTangent=-2.5249, tangentMode=RCTM_User, time=0.1408, value=0.9847}
          {arriveTangent=-0.0405, interpMode=RCIM_Cubic, leaveTangent=-0.0405, tangentMode=RCTM_User, time=1, value=0}
         ]
        }}})))
   &ScaleSpriteSize(Uniform Curve Sprite Scale=DataInterface() {propertyValues {Curve {
       defaultValue=0
       keys [
        {arriveTangent=9.0725, interpMode=RCIM_Cubic, leaveTangent=9.0725, tangentMode=RCTM_User, time=0, value=0}
        {arriveTangent=-2.5249, interpMode=RCIM_Cubic, leaveTangent=-2.5249, tangentMode=RCTM_User, time=0.1408, value=0.9847}
        {arriveTangent=-0.0405, interpMode=RCIM_Cubic, leaveTangent=-0.0405, tangentMode=RCTM_User, time=1, value=0}
       ]
      }}})
   &GravityForce(Gravity=&RandomRangeVector(Minimum=<0 0 -425>, Maximum=<0 0 -250>))
   &MeshRotationForce(Rotation=&RandomRangeVector(Minimum=<-15 0 0>, Maximum=<-25 0 0>))
   &Drag(
    Drag=&FloatFromCurve(
     FloatCurve=DataInterface() {propertyValues {Curve {keys [<at 0 0>,<at 0.8 1>]}}}
     Scale Curve=&RandomRangeFloat(Minimum=12, Maximum=5)
    )
    Rotational Drag:NiagaraFloat=&RandomRangeFloat(Minimum=2, Maximum=5)
    Use Linear Drag=true
    Use Rotational Drag=true
   )
   &SolveForcesAndVelocity(Clamp Velocity=false, Limit Acceleration=false)
   &ScaleColor(
    Scale Alpha=&FloatFromCurve(FloatCurve=DataInterface() {propertyValues {Curve {
        defaultValue=0
        keys [
         {arriveTangent=9.0725, interpMode=RCIM_Cubic, leaveTangent=9.0725, tangentMode=RCTM_User, time=0, value=0}
         {arriveTangent=-2.5249, interpMode=RCIM_Cubic, leaveTangent=-2.5249, tangentMode=RCTM_User, time=0.1408, value=0.9847}
         {arriveTangent=-0.0405, interpMode=RCIM_Cubic, leaveTangent=-0.0405, tangentMode=RCTM_User, time=1, value=0}
        ]
       }}})
    Scale RGB=&VectorFromFloat(Value=&FloatFromCurve(FloatCurve=DataInterface() {propertyValues {Curve {
         defaultValue=0
         keys [
          {arriveTangent=9.0725, interpMode=RCIM_Cubic, leaveTangent=9.0725, tangentMode=RCTM_User, time=0, value=0}
          {arriveTangent=-2.5249, interpMode=RCIM_Cubic, leaveTangent=-2.5249, tangentMode=RCTM_User, time=0.1408, value=0.9847}
          {arriveTangent=-0.0405, interpMode=RCIM_Cubic, leaveTangent=-0.0405, tangentMode=RCTM_User, time=1, value=0}
         ]
        }}}))
    ScaleA=true
    ScaleRGB=true
   )
   /SetVariables(
    Particles.InnerStarCO=5
    Particles.InnerStarSize:Vector3f=&VectorFromFloat(Value=&FloatFromCurve(
      FloatCurve=DataInterface() {propertyValues {Curve {
         defaultValue=0
         keys [
          {arriveTangent=9.0725, interpMode=RCIM_Cubic, leaveTangent=9.0725, tangentMode=RCTM_User, time=0, value=0}
          {arriveTangent=-2.5249, interpMode=RCIM_Cubic, leaveTangent=-2.5249, tangentMode=RCTM_User, time=0.1408, value=0.9847}
          {arriveTangent=-0.0405, interpMode=RCIM_Cubic, leaveTangent=-0.0405, tangentMode=RCTM_User, time=1, value=0}
         ]
        }}}
      Scale Curve=&RandomRangeFloat(Minimum=0.3, Maximum=0.4)
     ))
   ) {isSetParameters=true}
  ]
  rendererClasses [&NiagaraMeshRendererProperties]
  renderers [
   {rendererClass=&NiagaraMeshRendererProperties, rendererIndex=0}
   {rendererClass=&NiagaraMeshRendererProperties, rendererIndex=1}
  ]
 }
 ShootingStarRibbon {
  emitterSpawnScript [&InitializeAudioPlayer(Audio Player=DataInterface() {propertyValues {
      Attenuation=null
      Concurrency=null
      MaxPlaysPerTick=10
      SoundToPlay=&MSS_Emote_Foley_WishingStar_CometBys
      bAllowLoopingOneShotSounds=false
      bLimitPlaysPerTick=true
      bOnlyActiveDuringGameplay=false
      bStopWhenComponentIsDestroyed=true
      ConfigurationUserParameter {parameter {name=None, typeDefHandle {registeredTypeIndex=1}, varData []}}
      ParameterNames []
     }})]
  emitterUpdateScript [
   &EmitterState()
   &SpawnParticlesFromOtherEmitter(
    Attribute Reader=DataInterface() {propertyValues {EmitterBinding {bindingMode=Other, emitterName=ShootingStarHead}}}
    Spawn Rate=60
    Use Maximum Age=false
    Use Minimum Age=false
   )
  ]
  particleSpawnScript [
   &InitializeParticle(
    Color=<rgba 5 2.5132 0 1>
    Lifetime=1.2
    Ribbon Width=&RandomRangeFloat(Minimum=5, Maximum=8)
    Ribbon Width Mode=::NewEnumerator1
    UsePositionOffset=false
    Write Material Random=true
   )
   &SampleParticlesFromOtherEmitter(
    Age Sampling:ENiagara_DirectReadSamplingMode=::NewEnumerator0
    Applied Alpha Scale=0.3
    Applied Position Offset:Vector3f=<-2 0 0>
    Apply Sampled RGB=false
    Color Sampling=::NewEnumerator0
    Particle ID Sampling=::NewEnumerator1
   )
  ]
  particleUpdateScript [
   &ParticleState()
   &SolveForcesAndVelocity(Clamp Velocity=false, Limit Acceleration=false)
   &ScaleRibbonWidth(Ribbon Width Scale=&FloatFromCurve(FloatCurve=DataInterface() {propertyValues {Curve {
        defaultValue=1
        keys [
         {arriveTangent=-1.9314, interpMode=RCIM_Cubic, leaveTangent=-1.9314, tangentMode=RCTM_User, time=0, value=1}
         {interpMode=RCIM_Cubic, tangentMode=RCTM_User, time=1, value=0}
        ]
       }}}))
   &ScaleColor(
    Scale Alpha=&FloatFromCurve(FloatCurve=DataInterface() {propertyValues {Curve {
        defaultValue=1
        keys [
         {tangentMode=RCTM_User, time=0, value=1}
         {arriveTangent=-0.0405, interpMode=RCIM_Cubic, leaveTangent=-0.0405, tangentMode=RCTM_User, time=0.7009, value=1}
         {arriveTangent=-8.0089, interpMode=RCIM_Cubic, leaveTangent=-8.0089, tangentMode=RCTM_User, time=1, value=0}
        ]
       }}})
    Scale RGB=&VectorFromFloat(Value=&FloatFromCurve(FloatCurve=DataInterface() {propertyValues {Curve {
         defaultValue=0
         keys [
          {arriveTangent=9.0725, interpMode=RCIM_Cubic, leaveTangent=9.0725, tangentMode=RCTM_User, time=0, value=0}
          {arriveTangent=-2.5249, interpMode=RCIM_Cubic, leaveTangent=-2.5249, tangentMode=RCTM_User, time=0.1408, value=0.9847}
          {arriveTangent=-0.0405, interpMode=RCIM_Cubic, leaveTangent=-0.0405, tangentMode=RCTM_User, time=1, value=0}
         ]
        }}}))
    ScaleA=true
    ScaleRGB=true
   )
  ]
  rendererClasses [&NiagaraRibbonRendererProperties]
  renderers [{rendererClass=&NiagaraRibbonRendererProperties}]
 }
 StarGlowBurst {
  emitterUpdateScript [
   &EmitterState()
   &SpawnBurst_Instantaneous(Spawn Time=6.23, Use Spawn Probability=false, UseLoopCountLimit=false)
   &SpawnBurst_Instantaneous(Spawn Count=56, Spawn Group=1, Spawn Time=6.23, Use Spawn Probability=false, UseLoopCountLimit=false) {module=SpawnBurst_Instantaneous001}
   &SpawnBurst_Instantaneous(Spawn Group=2, Spawn Time=6.26, Use Spawn Probability=false, UseLoopCountLimit=false) {module=SpawnBurst_Instantaneous002}
  ]
  particleSpawnScript [
   /SetVariables(Particles.VisibilityTag=&SpawnGroup()) {isSetParameters=true}
   &InitializeParticle(
    Color=<rgba 1 1 1 0.7>
    Lifetime=&MaskFloatBySpawnGroup(
     Masked Float=0.4
     Passthrough Float=&MaskFloatBySpawnGroup(Masked Float=&RandomRangeFloat(Minimum=0.4, Maximum=1.2), Passthrough Float=1, Spawn Group Mask=1)
    )
    Mesh Scale Mode=::NewEnumerator3
    Mesh Uniform Scale=0.9
    Position Offset=<1 -27.5 40>
    Sprite Size Mode=::NewEnumerator3
    Uniform Sprite Size=72.5
    UsePositionOffset=true
    Write Material Random=true
   )
   &AddVelocity(
    Module.Use Falloff Distance=false
    Origin Offset=<1 -27.5 40>
    Override Seed=false
    Velocity Mode=::NewEnumerator1
    Velocity Speed=&MaskFloatBySpawnGroup(
     Masked Float=0
     Passthrough Float=&MaskFloatBySpawnGroup(Masked Float=&RandomRangeFloat(Minimum=45, Maximum=125), Passthrough Float=0, Spawn Group Mask=1)
    )
   )
   &CameraOffset(Camera Offset Amount=10)
   /SetVariables(Particles.StarBurstFacing:Vector3f=<0 1 0>) {isSetParameters=true}
  ]
  particleUpdateScript [
   &ParticleState()
   &ScaleSpriteSize(Uniform Curve Sprite Scale=DataInterface() {propertyValues {Curve {
       defaultValue=0
       keys [
        {arriveTangent=4.1785, interpMode=RCIM_Cubic, leaveTangent=4.1785, tangentMode=RCTM_User, time=0, value=0}
        {arriveTangent=1.1179, interpMode=RCIM_Cubic, leaveTangent=1.1179, tangentMode=RCTM_User, time=0.2042, value=0.6228}
        {arriveTangent=0.0224, interpMode=RCIM_Cubic, leaveTangent=0.0224, tangentMode=RCTM_User, time=1, value=1}
       ]
      }}})
   &ScaleMeshSize(Scale Factor=&VectorFromFloat(Value=&FloatFromCurve(FloatCurve=DataInterface() {propertyValues {Curve {keys [
          {arriveTangent=19.3094, interpMode=RCIM_Cubic, leaveTangent=19.3094, tangentMode=RCTM_User, time=0, value=0}
          {arriveTangent=0.5, interpMode=RCIM_Cubic, leaveTangent=0.5, time=0.075, value=1}
          {interpMode=RCIM_Cubic, leaveTangent=-0.1625, time=1, value=0.5}
         ]}}})))
   &Drag(
    Drag=&FloatFromCurve(
     FloatCurve=DataInterface() {propertyValues {Curve {keys [<at 0 0 RCIM_Cubic>, {arriveTangent=25.0356, interpMode=RCIM_Cubic, leaveTangent=25.0356, tangentMode=RCTM_User, time=0.1, value=1}]}}}
     Scale Curve=&RandomRangeFloat(Minimum=1.2, Maximum=1.9)
    )
    Use Linear Drag=true
    Use Rotational Drag=true
   )
   &ScaleColor(
    Linear Color Curve=DataInterface() {propertyValues {
      AlphaCurve {
       defaultValue=1
       keys [
        {arriveTangent=9.0725, interpMode=RCIM_Cubic, leaveTangent=9.0725, tangentMode=RCTM_User, time=0, value=0}
        {arriveTangent=-2.5249, interpMode=RCIM_Cubic, leaveTangent=-2.5249, tangentMode=RCTM_User, time=0.1408, value=0.9847}
        <at 0.35 1>
        {arriveTangent=-0.0405, interpMode=RCIM_Cubic, leaveTangent=-0.0405, tangentMode=RCTM_User, time=1, value=0}
       ]
      }
      BlueCurve {keys [<at 0 0>,<at 0.6 0>]}
      GreenCurve {keys [<at 0 2.7556>,<at 0.6 0.3033>]}
      RedCurve {keys [<at 0 5>,<at 0.6 0.599>]}
     }}
    Scale Mode=::NewEnumerator2
   )
   &SolveForcesAndVelocity(Clamp Velocity=false, Limit Acceleration=false)
   /SetVariables(
    Particles.BurstSparkleColor:LinearColor=&MultiplyLinearColor(
     Linear Color=&RandomRangeLinearColor(Minimum=<rgba 0.3273 0.8333 0 1>, Maximum=<rgba 1 0.8219 0 1>)
     Scale Factor=&Vector4FromCurve(Vector4Curve=DataInterface() {propertyValues {
        AlphaCurve {
         defaultValue=1
         keys [
          {arriveTangent=10.7395, interpMode=RCIM_Cubic, leaveTangent=10.7395, tangentMode=RCTM_User, time=0, value=3}
          {arriveTangent=-4.2154, arriveTangentWeight=0.1, interpMode=RCIM_Cubic, leaveTangent=12.8937, tangentMode=RCTM_Break, time=0.3, value=1}
          {
           arriveTangent=-15.3617
           arriveTangentWeight=0.0521
           interpMode=RCIM_Cubic
           leaveTangent=4.0988
           tangentMode=RCTM_Break
           time=0.4562
           value=1.0768
          }
          <at 1 0 RCIM_Cubic>
         ]
        }
        BlueCurve {
         defaultValue=1
         keys [<at 0 18 RCIM_Cubic>, {arriveTangent=-17.4, interpMode=RCIM_Cubic, leaveTangent=-17.4, time=0.3, value=11.8641}, <at 1 0.6 RCIM_Cubic>]
        }
        GreenCurve {
         defaultValue=1
         keys [<at 0 18 RCIM_Cubic>, {arriveTangent=-17.7202, interpMode=RCIM_Cubic, leaveTangent=-17.7202, time=0.3, value=9}, <at 1 0.2798 RCIM_Cubic>]
        }
        RedCurve {
         defaultValue=1
         keys [<at 0 18 RCIM_Cubic>, {arriveTangent=-17.7313, interpMode=RCIM_Cubic, leaveTangent=-17.7313, time=0.3, value=9}, <at 1 0.2688 RCIM_Cubic>]
        }
       }})
    )
    Particles.BurstSparkleSize:Vector2f=&Vector2DFromFloat(Value=&Multiply_Float(
      A=&RandomRangeFloat(Minimum=5, Maximum=8)
      B=&FloatFromCurve(FloatCurve=DataInterface() {propertyValues {Curve {
          defaultValue=0
          keys [
           {arriveTangent=9.0725, interpMode=RCIM_Cubic, leaveTangent=9.0725, tangentMode=RCTM_User, time=0, value=0}
           {arriveTangent=-2.5249, interpMode=RCIM_Cubic, leaveTangent=-2.5249, tangentMode=RCTM_User, time=0.1408, value=0.9847}
           {arriveTangent=-0.0405, interpMode=RCIM_Cubic, leaveTangent=-0.0405, tangentMode=RCTM_User, time=1, value=0}
          ]
         }}})
     ))
   ) {isSetParameters=true}
   /SetVariables(
    Particles.GlowBurstColor:LinearColor=&ColorFromCurve(ColorCurve=DataInterface() {propertyValues {
       AlphaCurve {
        defaultValue=1
        keys [
         {arriveTangent=9.0725, interpMode=RCIM_Cubic, leaveTangent=9.0725, tangentMode=RCTM_User, time=0, value=0}
         {arriveTangent=-2.5249, interpMode=RCIM_Cubic, leaveTangent=-2.5249, tangentMode=RCTM_User, time=0.1408, value=0.9847}
         {arriveTangent=-0.0405, interpMode=RCIM_Cubic, leaveTangent=-0.0405, tangentMode=RCTM_User, time=1, value=0}
        ]
       }
       BlueCurve {keys [<at 0 0>,<at 0.8462 0>]}
       GreenCurve {keys [<at 0 1.9469>,<at 0.8462 0.2894>]}
       RedCurve {keys [<at 0 5>,<at 0.8462 0.599>]}
      }})
    Particles.GlowBurstSize:Vector2f=&Vector2DFromFloat(Value=&Multiply_Float(
      A=85
      B=&FloatFromCurve(FloatCurve=DataInterface() {propertyValues {Curve {
          defaultValue=0
          keys [
           {arriveTangent=4.1785, interpMode=RCIM_Cubic, leaveTangent=4.1785, tangentMode=RCTM_User, time=0, value=0}
           {arriveTangent=1.1179, interpMode=RCIM_Cubic, leaveTangent=1.1179, tangentMode=RCTM_User, time=0.2042, value=0.6228}
           {arriveTangent=0.0224, interpMode=RCIM_Cubic, leaveTangent=0.0224, tangentMode=RCTM_User, time=1, value=1}
          ]
         }}})
     ))
   ) {isSetParameters=true}
   /SetVariables(
    Particles.RayBurstColor:LinearColor=&ColorFromCurve(ColorCurve=DataInterface() {propertyValues {
       AlphaCurve {
        defaultValue=1
        keys [
         {arriveTangent=9.0725, time=0, value=0}
         {arriveTangent=-2.5249, leaveTangent=-2.5249, time=0.15, value=0.3}
         <at 0.45 0.15>
         {arriveTangent=-0.0405, interpMode=RCIM_Cubic, leaveTangent=-0.0405, tangentMode=RCTM_User, time=1, value=0}
        ]
       }
       BlueCurve {keys [<at 0 0>,<at 0.8462 0>]}
       GreenCurve {keys [<at 0 1.9469>,<at 0.8462 0.2894>]}
       RedCurve {keys [<at 0 5>,<at 0.8462 0.599>]}
      }})
    Particles.RayBurstDP:Vector4f=&MakeVector4(
     X=&RandomRangeFloat(Minimum=0, Maximum=10)
     Y=&RandomRangeFloat(Minimum=2.5, Maximum=4.5)
     Z=&RandomRangeFloat(Minimum=0.8, Maximum=1.6)
     W=0
    )
   ) {isSetParameters=true}
  ]
  rendererClasses [&NiagaraSpriteRendererProperties, &NiagaraMeshRendererProperties]
  renderers [
   {rendererClass=&NiagaraSpriteRendererProperties}
   {rendererClass=&NiagaraSpriteRendererProperties, rendererIndex=1}
   {rendererClass=&NiagaraSpriteRendererProperties, rendererIndex=2}
   {rendererClass=&NiagaraMeshRendererProperties, rendererIndex=3}
  ]
 }
 StarPickupBurstOutRibbon {
  emitterUpdateScript [
   &EmitterState()
   &SpawnRate(SpawnGroup=4, SpawnRate=&Multiply_Float(A=120, B=$User.StarParticles), Use Spawn Probability=false)
   &SpawnBurst_Instantaneous(Spawn Time=2.9, Use Spawn Probability=false, UseLoopCountLimit=false)
   &SpawnBurst_Instantaneous(Spawn Group=1, Spawn Time=8.26, Use Spawn Probability=false, UseLoopCountLimit=false) {module=SpawnBurst_Instantaneous001}
   &SpawnBurst_Instantaneous(Spawn Count=32, Spawn Group=2, Spawn Time=8.26, Use Spawn Probability=false, UseLoopCountLimit=false) {module=SpawnBurst_Instantaneous002}
   &SpawnBurst_Instantaneous(Spawn Group=3, Spawn Time=8.26, Use Spawn Probability=false, UseLoopCountLimit=false) {module=SpawnBurst_Instantaneous003}
   &SpawnBurst_Instantaneous(Spawn Group=5, Spawn Time=2.3, Use Spawn Probability=false, UseLoopCountLimit=false) {module=SpawnBurst_Instantaneous004}
  ]
  particleSpawnScript [
   /SetVariables(Particles.VisibilityTag=&SpawnGroup()) {isSetParameters=true}
   &InitializeParticle(
    Color=<rgba 1 1 0 0.702>
    Lifetime=&MaskFloatBySpawnGroup(
     Masked Float=0.28
     Passthrough Float=&MaskFloatBySpawnGroup(
      Masked Float=0.28
      Passthrough Float=&MaskFloatBySpawnGroup(
       Masked Float=&RandomRangeFloat(Minimum=0.3, Maximum=1)
       Passthrough Float=&MaskFloatBySpawnGroup(
        Masked Float=0.28
        Passthrough Float=&MaskFloatBySpawnGroup(Masked Float=&RandomRangeFloat(Minimum=0.3, Maximum=0.8), Passthrough Float=6.15, Spawn Group Mask=4)
        Spawn Group Mask=3
       )
       Spawn Group Mask=2
      )
      Spawn Group Mask=1
     )
    )
    Position Offset=&MaskVectorBySpawnGroup(Masked Vector=<5 0 -10>, Passthrough Vector=<0 0 0>)
    Sprite Size Mode=::NewEnumerator3
    Uniform Sprite Size=80
    UsePositionOffset=true
    Write Material Random=true
   )
   &SkeletalMeshLocation(
    Bone Sampling Mode=::NewEnumerator4
    MaskSpawnGroup=false
    Sampled Position Offset=&MaskVectorBySpawnGroup(Masked Vector=<-10 0 0>, Passthrough Vector=<0 -5 0>)
    Sampled Position Offset Coordinate Space:ENiagaraSkelMeshTransforms=::NewEnumerator2
    Skeletal Mesh=$System.Skeletal Mesh
   )
   &ShapeLocation(
    MaskSpawnGroup=false
    Override Seed=false
    Sphere Radius=&MaskFloatBySpawnGroup(
     Masked Float=0
     Passthrough Float=&MaskFloatBySpawnGroup(
      Masked Float=0
      Passthrough Float=&MaskFloatBySpawnGroup(
       Masked Float=4
       Passthrough Float=&MaskFloatBySpawnGroup(
        Masked Float=0
        Passthrough Float=&MaskFloatBySpawnGroup(Masked Float=4, Passthrough Float=1, Spawn Group Mask=4)
        Spawn Group Mask=3
       )
       Spawn Group Mask=2
      )
      Spawn Group Mask=1
     )
    )
    UseOwnerScale=true
   )
   &AddVelocity(
    Module.Use Falloff Distance=false
    Override Seed=false
    Velocity Mode=::NewEnumerator1
    Velocity Origin=$Particles.ShapeLocation.ShapePosition
    Velocity Speed=&MaskFloatBySpawnGroup(
     Masked Float=0
     Passthrough Float=&MaskFloatBySpawnGroup(
      Masked Float=0
      Passthrough Float=&MaskFloatBySpawnGroup(
       Masked Float=&RandomRangeFloat(Minimum=45, Maximum=90)
       Passthrough Float=&MaskFloatBySpawnGroup(
        Masked Float=0
        Passthrough Float=&MaskFloatBySpawnGroup(Masked Float=0, Passthrough Float=0, Spawn Group Mask=4)
        Spawn Group Mask=3
       )
       Spawn Group Mask=2
      )
      Spawn Group Mask=1
     )
    )
   )
   &AddVelocity(
    Cone Angle=65
    Cone Axis=&VectorFromCurve(
     VectorCurve=DataInterface() {propertyValues {
       XCurve {keys [<at 0 1>, {arriveTangent=-0.2, leaveTangent=-0.2, time=4.9, value=1}, <at 5 0>]}
       YCurve {keys [<at 0 0>,<at 1 0>]}
       ZCurve {keys [<at 0 0.4>,<at 4.9 0.4>,<at 5 0>,<at 7.9 0>,<at 8 -1>]}
      }}
     CurveIndex=$Emitter.LoopedAge
    )
    Module.UseConeFalloff=false
    Module.UseVelDistribution=false
    Override Seed=false
    Velocity Mode=::NewEnumerator2
    Velocity Speed=&MaskFloatBySpawnGroup(
     Masked Float=0
     Passthrough Float=&MaskFloatBySpawnGroup(
      Masked Float=0
      Passthrough Float=&MaskFloatBySpawnGroup(
       Masked Float=0
       Passthrough Float=&MaskFloatBySpawnGroup(
        Masked Float=0
        Passthrough Float=&MaskFloatBySpawnGroup(Masked Float=&RandomRangeFloat(Minimum=80, Maximum=145), Passthrough Float=0, Spawn Group Mask=4)
        Spawn Group Mask=3
       )
       Spawn Group Mask=2
      )
      Spawn Group Mask=1
     )
    )
   ) {module=AddVelocity001}
   &CameraOffset(Camera Offset Amount=&MaskFloat(Mask Source=0, Mask Parameter=?, Mask=5, Masked Float=0, Passthrough Float=150)) {enabled=false}
  ]
  particleUpdateScript [
   &ParticleState()
   &SkeletalMeshLocation(Bone Sampling Mode=::NewEnumerator4, MaskSpawnGroup=true, Position Sampling=::NewEnumerator2, Skeletal Mesh=$System.Skeletal Mesh) {module=SkeletalMeshLocation001}
   &CurlNoiseForce(
    Bias Noise=false
    Noise Frequency=12
    Noise Strength=&MaskFloatBySpawnGroup(
     Masked Float=0
     Passthrough Float=&MaskFloatBySpawnGroup(
      Masked Float=0
      Passthrough Float=&MaskFloatBySpawnGroup(
       Masked Float=0
       Passthrough Float=&MaskFloatBySpawnGroup(
        Masked Float=0
        Passthrough Float=&MaskFloatBySpawnGroup(Masked Float=&RandomRangeFloat(Minimum=75, Maximum=185), Passthrough Float=0, Spawn Group Mask=4)
        Spawn Group Mask=3
       )
       Spawn Group Mask=2
      )
      Spawn Group Mask=1
     )
    )
    Pan Noise=false
   )
   &GravityForce(Gravity=&MaskVectorBySpawnGroup(
     Masked Vector=<0 0 0>
     Passthrough Vector=&MaskVectorBySpawnGroup(
      Masked Vector=<0 0 0>
      Passthrough Vector=&MaskVectorBySpawnGroup(
       Masked Vector=<0 0 0>
       Passthrough Vector=&MaskVectorBySpawnGroup(
        Masked Vector=<0 0 0>
        Passthrough Vector=&MaskVectorBySpawnGroup(
         Masked Vector=&RandomRangeVector(Minimum=<0 0 -50>, Maximum=<0 0 -80>)
         Passthrough Vector=<0 0 0>
         Spawn Group Mask=4
        )
        Spawn Group Mask=3
       )
       Spawn Group Mask=2
      )
      Spawn Group Mask=1
     )
    ))
   &Drag(
    Drag=&MaskFloatBySpawnGroup(
     Masked Float=&RandomRangeFloat(Minimum=1, Maximum=1.2)
     Passthrough Float=&MaskFloatBySpawnGroup(
      Masked Float=&RandomRangeFloat(Minimum=1, Maximum=1.2)
      Passthrough Float=&MaskFloatBySpawnGroup(
       Masked Float=&RandomRangeFloat(Minimum=1, Maximum=1.2)
       Passthrough Float=&MaskFloatBySpawnGroup(
        Masked Float=&RandomRangeFloat(Minimum=1, Maximum=1.2)
        Passthrough Float=&RandomRangeFloat(Minimum=1.6, Maximum=2.2)
        Spawn Group Mask=3
       )
       Spawn Group Mask=2
      )
      Spawn Group Mask=1
     )
    )
    Use Linear Drag=true
    Use Rotational Drag=true
   )
   &SpriteRotationRate(Rotation Rate=&MaskFloatBySpawnGroup(
     Masked Float=0
     Passthrough Float=&MaskFloatBySpawnGroup(
      Masked Float=15
      Passthrough Float=&MaskFloatBySpawnGroup(Masked Float=0, Passthrough Float=0, Spawn Group Mask=2)
      Spawn Group Mask=1
     )
    ))
   &ScaleSpriteSize(Uniform Curve Sprite Scale=DataInterface() {propertyValues {Curve {
       defaultValue=0
       keys [
        {arriveTangent=4.1785, interpMode=RCIM_Cubic, leaveTangent=4.1785, tangentMode=RCTM_User, time=0, value=0}
        {arriveTangent=1.1179, interpMode=RCIM_Cubic, leaveTangent=1.1179, tangentMode=RCTM_User, time=0.2042, value=0.6228}
        {arriveTangent=0.0224, interpMode=RCIM_Cubic, leaveTangent=0.0224, tangentMode=RCTM_User, time=1, value=1}
       ]
      }}})
   &ScaleColor(
    Scale Alpha=&FloatFromCurve(FloatCurve=DataInterface() {propertyValues {Curve {
        defaultValue=1
        keys [
         {arriveTangent=9.0725, interpMode=RCIM_Cubic, leaveTangent=9.0725, tangentMode=RCTM_User, time=0, value=0}
         {arriveTangent=-2.5249, interpMode=RCIM_Cubic, leaveTangent=-2.5249, tangentMode=RCTM_User, time=0.1408, value=0.9847}
         {arriveTangent=-0.0405, interpMode=RCIM_Cubic, leaveTangent=-0.0405, tangentMode=RCTM_User, time=1, value=0}
        ]
       }}})
    ScaleA=true
    ScaleRGB=true
   )
   /SetVariables(
    Particles.GlowColor=/Niagara/DynamicInputs/LinearColor/MakeLinearColorFromVector4(Vector:Vector4f=&Multiply_Vector4(
      A=&MakeVector4_FromLinearColor(LinearColor=&ColorFromCurve(ColorCurve=DataInterface() {propertyValues {
          AlphaCurve {
           defaultValue=1
           keys [
            {arriveTangent=9.0725, interpMode=RCIM_Cubic, leaveTangent=9.0725, tangentMode=RCTM_User, time=0, value=0}
            {arriveTangent=-2.5249, arriveTangentWeight=0.0469, interpMode=RCIM_Cubic, leaveTangent=-2.5249, tangentMode=RCTM_User, time=0.15, value=0.6}
            {arriveTangent=-0.0405, arriveTangentWeight=0.2864, interpMode=RCIM_Cubic, leaveTangent=-0.0405, tangentMode=RCTM_User, time=1, value=0}
           ]
          }
          BlueCurve {keys [<at 0 0 RCIM_Cubic>]}
          GreenCurve {keys [<at 0 10.5223 RCIM_Cubic>]}
          RedCurve {keys [<at 0 15 RCIM_Cubic>]}
         }}))
      B=<1 1 1 1>
     ))
    Particles.GlowSize=&Vector2DFromFloat(Value=&FloatFromCurve(
      FloatCurve=DataInterface() {propertyValues {Curve {
         defaultValue=0
         keys [
          {arriveTangent=4.1785, interpMode=RCIM_Cubic, leaveTangent=4.1785, tangentMode=RCTM_User, time=0, value=0}
          {arriveTangent=1.1179, interpMode=RCIM_Cubic, leaveTangent=1.1179, tangentMode=RCTM_User, time=0.2042, value=0.6228}
          {arriveTangent=0.0224, interpMode=RCIM_Cubic, leaveTangent=0.0224, tangentMode=RCTM_User, time=1, value=1}
         ]
        }}}
      Scale Curve=120
     ))
   ) {isSetParameters=true}
   /SetVariables(
    Particles.StarOutColor:LinearColor=&ColorFromCurve(ColorCurve=DataInterface() {propertyValues {
       AlphaCurve {
        defaultValue=0
        keys [
         <at 0 1>
         {arriveTangent=-0.0405, arriveTangentWeight=0.2336, interpMode=RCIM_Cubic, leaveTangent=-0.0405, tangentMode=RCTM_User, time=0.7009, value=1}
         {arriveTangent=-8.0089, arriveTangentWeight=0.0997, interpMode=RCIM_Cubic, leaveTangent=-8.0089, tangentMode=RCTM_User, time=1, value=0}
        ]
       }
       BlueCurve {keys [<at 0 0.2245 RCIM_Cubic>]}
       GreenCurve {keys [<at 0 0.7341 RCIM_Cubic>]}
       RedCurve {keys [<at 0 1 RCIM_Cubic>]}
      }})
    Particles.StarOutSize:Vector2f=&Vector2DFromFloat(Value=&FloatFromCurve(
      FloatCurve=DataInterface() {propertyValues {Curve {
         defaultValue=0
         keys [
          {interpMode=RCIM_Cubic, tangentMode=RCTM_User, time=0, value=0}
          {interpMode=RCIM_Cubic, tangentMode=RCTM_User, time=0.2, value=1}
          {tangentMode=RCTM_User, time=1, value=0}
         ]
        }}}
      Scale Curve=35
     ))
   ) {isSetParameters=true}
   /SetVariables(
    Particles.BurstSparkleColor:LinearColor=&MultiplyLinearColor(
     Linear Color=&RandomRangeLinearColor(Minimum=<rgba 0.5309 0.5625 0 1>, Maximum=<rgba 2 1.2461 0.0043 1>)
     Scale Factor=&Vector4FromCurve(Vector4Curve=DataInterface() {propertyValues {
        AlphaCurve {
         defaultValue=1
         keys [
          {arriveTangent=10.7395, interpMode=RCIM_Cubic, leaveTangent=10.7395, tangentMode=RCTM_User, time=0, value=3}
          {arriveTangent=-4.2154, arriveTangentWeight=0.1, interpMode=RCIM_Cubic, leaveTangent=12.8937, tangentMode=RCTM_Break, time=0.3, value=1}
          {
           arriveTangent=-15.3617
           arriveTangentWeight=0.0521
           interpMode=RCIM_Cubic
           leaveTangent=4.0988
           tangentMode=RCTM_Break
           time=0.4562
           value=1.0768
          }
          <at 1 0 RCIM_Cubic>
         ]
        }
        BlueCurve {
         defaultValue=1
         keys [<at 0 10 RCIM_Cubic>, {arriveTangent=-9.4, interpMode=RCIM_Cubic, leaveTangent=-9.4, time=0.3, value=8}, <at 1 0.6 RCIM_Cubic>]
        }
        GreenCurve {
         defaultValue=1
         keys [<at 0 10 RCIM_Cubic>, {arriveTangent=-9.7202, interpMode=RCIM_Cubic, leaveTangent=-9.7202, time=0.3, value=6.0687}, <at 1 0.2798 RCIM_Cubic>]
        }
        RedCurve {
         defaultValue=1
         keys [<at 0 10 RCIM_Cubic>, {arriveTangent=-9.7312, interpMode=RCIM_Cubic, leaveTangent=-9.7312, time=0.3, value=6.0687}, <at 1 0.2688 RCIM_Cubic>]
        }
       }})
    )
    Particles.BurstSparkleSize:Vector2f=&Vector2DFromFloat(Value=&Multiply_Float(
      A=&RandomRangeFloat(Minimum=5, Maximum=8)
      B=&FloatFromCurve(FloatCurve=DataInterface() {propertyValues {Curve {
          defaultValue=0
          keys [
           {arriveTangent=9.0725, interpMode=RCIM_Cubic, leaveTangent=9.0725, tangentMode=RCTM_User, time=0, value=0}
           {arriveTangent=-2.5249, interpMode=RCIM_Cubic, leaveTangent=-2.5249, tangentMode=RCTM_User, time=0.1408, value=0.9847}
           {arriveTangent=-0.0405, interpMode=RCIM_Cubic, leaveTangent=-0.0405, tangentMode=RCTM_User, time=1, value=0}
          ]
         }}})
     ))
   ) {isSetParameters=true}
   /SetVariables(
    Particles.GlowOutColor:LinearColor=&ColorFromCurve(ColorCurve=DataInterface() {propertyValues {
       AlphaCurve {
        defaultValue=1
        keys [
         {arriveTangent=9.0725, interpMode=RCIM_Cubic, leaveTangent=9.0725, tangentMode=RCTM_User, time=0, value=0}
         {arriveTangent=-2.5249, arriveTangentWeight=0.0469, interpMode=RCIM_Cubic, leaveTangent=-2.5249, tangentMode=RCTM_User, time=0.15, value=0.3}
         {arriveTangent=-0.0405, arriveTangentWeight=0.2864, interpMode=RCIM_Cubic, leaveTangent=-0.0405, tangentMode=RCTM_User, time=1, value=0}
        ]
       }
       BlueCurve {keys [<at 0 0 RCIM_Cubic>]}
       GreenCurve {keys [<at 0 8.9863 RCIM_Cubic>]}
       RedCurve {keys [<at 0 15 RCIM_Cubic>]}
      }})
    Particles.GlowOutSize:Vector2f=&Vector2DFromFloat(Value=&FloatFromCurve(
      FloatCurve=DataInterface() {propertyValues {Curve {
         defaultValue=0
         keys [
          {arriveTangent=4.1785, interpMode=RCIM_Cubic, leaveTangent=4.1785, tangentMode=RCTM_User, time=0, value=0}
          {arriveTangent=1.1179, interpMode=RCIM_Cubic, leaveTangent=1.1179, tangentMode=RCTM_User, time=0.2042, value=0.6228}
          {arriveTangent=0.0224, interpMode=RCIM_Cubic, leaveTangent=0.0224, tangentMode=RCTM_User, time=1, value=1}
         ]
        }}}
      Scale Curve=65
     ))
   ) {isSetParameters=true}
   /SetVariables(
    Particles.StarRibbonColor:LinearColor=&Multiply_LinearColorByFloat(
     Linear Color=&RandomRangeLinearColor(Minimum=<rgba 5.9625 8 0 1>, Maximum=<rgba 8 3.1093 0 1>)
     Float=&FloatFromCurve(FloatCurve=DataInterface() {propertyValues {Curve {
         defaultValue=1
         keys [
          {arriveTangent=9.0725, interpMode=RCIM_Cubic, leaveTangent=9.0725, tangentMode=RCTM_User, time=0, value=0}
          {arriveTangent=-2.5249, interpMode=RCIM_Cubic, leaveTangent=-2.5249, tangentMode=RCTM_User, time=0.1408, value=0.9847}
          {arriveTangent=-0.0405, interpMode=RCIM_Cubic, leaveTangent=-0.0405, tangentMode=RCTM_User, time=1, value=0}
         ]
        }}})
    )
    Particles.StarRibbonSize:Vector2f=&Vector2DFromFloat(Value=&FloatFromCurve(
      FloatCurve=DataInterface() {propertyValues {Curve {
         defaultValue=1
         keys [
          {tangentMode=RCTM_User, time=0, value=1}
          {arriveTangent=-0.0405, interpMode=RCIM_Cubic, leaveTangent=-0.0405, tangentMode=RCTM_User, time=0.7009, value=1}
          {arriveTangent=-8.0089, interpMode=RCIM_Cubic, leaveTangent=-8.0089, tangentMode=RCTM_User, time=1, value=0}
         ]
        }}}
      Scale Curve=&RandomRangeFloat(Minimum=8, Maximum=16)
     ))
   ) {isSetParameters=true}
   /SetVariables(
    Particles.StarGlowCO=27
    Particles.StarGlowColor:LinearColor=&ColorFromCurve(ColorCurve=DataInterface() {propertyValues {
       AlphaCurve {
        defaultValue=1
        keys [
         {arriveTangent=8.1785, interpMode=RCIM_Cubic, leaveTangent=8.1785, tangentMode=RCTM_User, time=0, value=0}
         {arriveTangentWeight=0.0667, interpMode=RCIM_Cubic, tangentMode=RCTM_User, time=0.15, value=0.4}
         <at 0.55 0.4>
         <at 0.625 1>
         {arriveTangent=0.0566, arriveTangentWeight=0.25, interpMode=RCIM_Cubic, leaveTangent=0.0566, tangentMode=RCTM_User, time=0.95, value=0.6}
         <at 1 0>
        ]
       }
       BlueCurve {
        defaultValue=1
        keys [
         <at 0 0 RCIM_Cubic>
         {arriveTangent=-0.2628, arriveTangentWeight=0.1333, interpMode=RCIM_Cubic, leaveTangent=-0.2628, tangentMode=RCTM_User, time=0.4, value=0}
         <at 0.625 0 RCIM_Cubic>
         {arriveTangent=0.0636, arriveTangentWeight=0.0083, interpMode=RCIM_Cubic, leaveTangent=0.0636, tangentMode=RCTM_User, time=0.65, value=0}
         <at 1 0 RCIM_Cubic>
        ]
       }
       GreenCurve {
        defaultValue=1
        keys [
         <at 0 0.563 RCIM_Cubic>
         {
          arriveTangent=-0.2628
          arriveTangentWeight=0.1333
          interpMode=RCIM_Cubic
          leaveTangent=-0.2628
          tangentMode=RCTM_User
          time=0.4
          value=0.5242
         }
         {arriveTangent=0.9188, interpMode=RCIM_Cubic, leaveTangent=0.9188, time=0.625, value=1.7486}
         {arriveTangent=0.0636, arriveTangentWeight=0.0083, interpMode=RCIM_Cubic, leaveTangent=0.0636, tangentMode=RCTM_User, time=0.65, value=0.7539}
         <at 1 0.6644 RCIM_Cubic>
        ]
       }
       RedCurve {
        defaultValue=1
        keys [
         <at 0 1 RCIM_Cubic>
         {arriveTangent=-0.2628, arriveTangentWeight=0.1333, interpMode=RCIM_Cubic, leaveTangent=-0.2628, tangentMode=RCTM_User, time=0.4, value=1}
         {arriveTangent=1.8, interpMode=RCIM_Cubic, leaveTangent=1.8, time=0.625, value=3}
         {arriveTangent=0.0636, arriveTangentWeight=0.0083, interpMode=RCIM_Cubic, leaveTangent=0.0636, tangentMode=RCTM_User, time=0.65, value=1.45}
         <at 1 1.45 RCIM_Cubic>
        ]
       }
      }})
    Particles.StarGlowSize:Vector2f=&Vector2DFromFloat(Value=&FloatFromCurve(
      FloatCurve=DataInterface() {propertyValues {Curve {
         defaultValue=0
         keys [
          {arriveTangent=8.1785, interpMode=RCIM_Cubic, leaveTangent=8.1785, tangentMode=RCTM_User, time=0, value=0}
          {arriveTangent=-0.0275, interpMode=RCIM_Cubic, leaveTangent=-0.0275, tangentMode=RCTM_User, time=0.15, value=1.25}
          {interpMode=RCIM_Cubic, tangentMode=RCTM_User, time=0.2, value=1}
          {arriveTangent=0.0566, interpMode=RCIM_Cubic, leaveTangent=0.0566, tangentMode=RCTM_User, time=0.9, value=1}
          {arriveTangent=-10.7112, tangentMode=RCTM_User, time=1, value=0}
         ]
        }}}
      Scale Curve=45
     ))
   ) {isSetParameters=true}
   &SolveForcesAndVelocity(Clamp Velocity=false, Limit Acceleration=false)
   &FN_Light_Attributes(
    Color=<rgba 1 0.6642 0 1>
    Create Color Attribute=true
    Create Exponent Attribute=true
    Create Radius Attribute=true
    Exponent=4
    Radius=80
    Scale Color=&FloatFromCurve(FloatCurve=DataInterface() {propertyValues {Curve {keys [
         {arriveTangent=16.2251, interpMode=RCIM_Cubic, leaveTangent=16.2251, tangentMode=RCTM_User, time=0, value=0}
         {arriveTangent=0.9068, interpMode=RCIM_Cubic, leaveTangent=0.9068, tangentMode=RCTM_User, time=0.1, value=1}
         {arriveTangent=-0.1268, interpMode=RCIM_Cubic, leaveTangent=-0.1268, tangentMode=RCTM_User, time=0.6, value=4}
         {arriveTangent=0.0636, interpMode=RCIM_Cubic, leaveTangent=0.0636, tangentMode=RCTM_User, time=0.65, value=1.45}
         {arriveTangent=0.0657, interpMode=RCIM_Cubic, leaveTangent=0.0657, tangentMode=RCTM_User, time=0.9, value=1.45}
         <at 1 0 RCIM_Cubic>
        ]}}})
   )
  ]
  rendererClasses [&NiagaraSpriteRendererProperties, &NiagaraLightRendererProperties]
  renderers [
   {rendererClass=&NiagaraSpriteRendererProperties}
   {rendererClass=&NiagaraSpriteRendererProperties, rendererIndex=1}
   {rendererClass=&NiagaraSpriteRendererProperties, rendererIndex=2}
   {rendererClass=&NiagaraSpriteRendererProperties, rendererIndex=3}
   {rendererClass=&NiagaraSpriteRendererProperties, rendererIndex=4}
   {rendererClass=&NiagaraSpriteRendererProperties, rendererIndex=5}
   {rendererClass=&NiagaraSpriteRendererProperties, rendererIndex=6}
   {rendererClass=&NiagaraLightRendererProperties, rendererIndex=7}
  ]
 }
 StarRibbon {
  emitterUpdateScript [
   &EmitterState()
   &SpawnRate(SpawnRate=&Multiply_Float(A=60, B=$User.Trail1), Use Spawn Probability=false)
   &SpawnRate(SpawnGroup=1, SpawnRate=&Multiply_Float(A=60, B=$User.Trail2), Use Spawn Probability=false) {module=SpawnRate001}
  ]
  particleSpawnScript [
   &InitializeParticle(
    Color=&ColorFromCurve(
     ColorCurve=DataInterface() {propertyValues {
       AlphaCurve {keys [<at 0 1 RCIM_Cubic>]}
       BlueCurve {keys [<at 0 0>,<at 1 0 RCIM_Cubic>,<at 1 0>]}
       GreenCurve {keys [<at 0 5.3231>, {arriveTangent=4.6769, interpMode=RCIM_Cubic, leaveTangent=4.6769, time=1, value=2.6616}, <at 1 10>]}
       RedCurve {keys [<at 0 8>, {arriveTangent=7, interpMode=RCIM_Cubic, leaveTangent=7, time=1, value=4}, <at 1 15>]}
      }}
     CurveIndex=$Emitter.Age
    )
    Lifetime=&MaskFloatBySpawnGroup(Masked Float=0.6, Passthrough Float=0.45)
    Ribbon Facing Mode=::NewEnumerator1
    Ribbon Facing Vector:Vector3f=$Engine.Owner.SystemZAxis
    Ribbon Width=&MaskFloatBySpawnGroup(Masked Float=10, Passthrough Float=24)
    Ribbon Width Mode=::NewEnumerator1
    UsePositionOffset=false
    Write Material Random=true
   )
   &SkeletalMeshLocation(
    Bone Sampling Mode=::NewEnumerator4
    MaskSpawnGroup=false
    Sampled Position Offset=&MaskVectorBySpawnGroup(Masked Vector=<0 0 -5>, Passthrough Vector=<0 0 5>)
    Skeletal Mesh=$System.Skeletal Mesh
   )
  ]
  particleUpdateScript [
   &ParticleState()
   &ScaleRibbonWidth(Ribbon Width Scale=&FloatFromCurve(FloatCurve=DataInterface() {propertyValues {Curve {
        defaultValue=0
        keys [
         {arriveTangent=-2.2912, interpMode=RCIM_Cubic, leaveTangent=-2.2912, tangentMode=RCTM_User, time=0, value=1}
         {interpMode=RCIM_Cubic, leaveTangent=-2.4754, time=1, value=0}
        ]
       }}}))
   &ScaleColor(
    Scale Alpha=&FloatFromCurve(
     FloatCurve=DataInterface() {propertyValues {Curve {
        defaultValue=0
        keys [
         {tangentMode=RCTM_User, time=0, value=1}
         {arriveTangent=-0.0405, interpMode=RCIM_Cubic, leaveTangent=-0.0405, tangentMode=RCTM_User, time=0.4, value=1}
         {arriveTangent=-2.4873, interpMode=RCIM_Cubic, leaveTangent=-2.4873, tangentMode=RCTM_User, time=1, value=0}
        ]
       }}}
     Scale Curve=0.3
    )
    ScaleA=true
    ScaleRGB=true
   )
   &CurlNoiseForce(
    Bias Noise=true
    Bias Noise Field:Vector3f=<0 0 -1>
    Noise Frequency=25
    Noise Strength=&MaskFloatBySpawnGroup(
     Masked Float=0
     Passthrough Float=&FloatFromCurve(
      FloatCurve=DataInterface() {propertyValues {Curve {keys [<at 0 0 RCIM_Cubic>, {arriveTangent=5.5436, interpMode=RCIM_Cubic, leaveTangent=5.5436, tangentMode=RCTM_User, time=0.2, value=1}]}}}
      Scale Curve=125
     )
    )
    Pan Noise=true
   )
   &Drag(Drag=2, Use Linear Drag=true, Use Rotational Drag=true)
   &InheritSourceMovement(ApplyForce=false, ApplyPosition=true, ApplyPositionFalloff=false)
   &SolveForcesAndVelocity(Clamp Velocity=false, Limit Acceleration=false)
  ]
  rendererClasses [&NiagaraRibbonRendererProperties]
  renderers [{rendererClass=&NiagaraRibbonRendererProperties}]
 }
}
summary {
 systemName=NS_Emote_WishingStar_Star
 emitters [
  {emitterName=StarGlowBurst, rendererClasses [&NiagaraSpriteRendererProperties, &NiagaraMeshRendererProperties]}
  {emitterName=StarRibbon, rendererClasses [&NiagaraRibbonRendererProperties]}
  {
   emitterName=StarPickupBurstOutRibbon
   rendererClasses [&NiagaraSpriteRendererProperties, &NiagaraLightRendererProperties]
  }
  {emitterName=ShootingStarHead, rendererClasses [&NiagaraMeshRendererProperties]}
  {emitterName=ShootingStarRibbon, rendererClasses [&NiagaraRibbonRendererProperties]}
  {emitterName=Galaxy, rendererClasses []}
 ]
 userVariables [
  User.StarParticles {type=NiagaraFloat, defaultValue=0, description=""}
  User.Trail1 {type=NiagaraFloat, defaultValue=0, description=""}
  User.Trail2 {type=NiagaraFloat, defaultValue=0, description=""}
 ]
}
